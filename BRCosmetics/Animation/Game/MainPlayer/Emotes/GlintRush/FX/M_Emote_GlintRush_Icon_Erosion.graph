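# Translucent unlit Niagara sprite material that shows the GlintRush emote icon, tinted by particle color, with an eroding wipe.
# Dynamic parameters blend between linear U and symmetric V wipes, set edge hardness, and add triangle-texture distortion.
# The optionally inverted mask drives a three-way lerp through the erosion texture, then multiplies icon and particle alpha.
# ErosionTileY and SubUVY are unused, as both X values feed both axes.
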Material /BRCosmetics/Animation/Game/MainPlayer/Emotes/GlintRush/FX/M_Emote_GlintRush_Icon_Erosion
graph {
 properties {
  bAutomaticallySetUsageInEditor=false
  bUsedWithNiagaraSprites=true
  blendMode=BLEND_Translucent
  shadingModel=MSM_Unlit
  translucencyLightingMode=TLM_SurfacePerPixelLighting
  usedShadingModels=Unlit
 }
 nodes [
  n1 = TextureSampleParameter2D() {parameterName=Icon Texture, texture=&/T__Emote_GlintRush_Number_New}
  n2 = ParticleColor()
  MaterialFunctionCall(Emissive=Multiply(A=@n1.RGB, B=@n2.RGB)) {materialFunction=&MF_FX_EmissiveAdjustments}
  n3 = Constant()
  n4 = TextureCoordinate()
  n5 = DynamicParameter() {paramNames [ErosionTileX, ErosionTileY, ErosionWipe, ErosionDistortionStr]}
  n6 = TextureSampleParameter2D(UVs=Multiply(A=@n4, B=AppendVector(A=@n5.ErosionTileX, B=@n5.ErosionTileX))) {parameterName=Erosion Texture, texture=/Game/Global/Textures/Default/Masks_and_Gradients/T-BattlePass-Tile-Triangles}
  n7 = Constant() {r=1}
  n8 = DynamicParameter() {parameterIndex=1, paramNames [ErosionWipeType, ErosionEdgeHardness, SubUVX, SubUVY]}
  n9 = Frac(None=Multiply(A=@n4, B=AppendVector(A=@n8.SubUVX, B=@n8.SubUVX)) {constB=2})
  n10 = Add(A=ComponentMask(None=@n9) {r=true}, B=@n5.ErosionWipe) {constB=-1}
  n11 = ComponentMask(None=@n9) {g=true}
  n12 = Multiply(A=Add(A=OneMinus(None=@n11), B=@n5.ErosionWipe), B=Add(A=@n11, B=@n5.ErosionWipe))
  n13 = Add(
   A=Multiply(A=@n6.R, B=@n5.ErosionDistortionStr) {constB=0.2}
   B=Multiply(A=LinearInterpolate(A=@n10, B=@n12, Alpha=@n8.ErosionWipeType), B=@n8.ErosionEdgeHardness)
  )
  n14 = LinearInterpolate(A=@n13, B=OneMinus(None=Clamp(None=@n13)), Alpha=ScalarParameter() {parameterName=InvertErosion})
  Multiply(
   A=Clamp(None=MaterialFunctionCall(1=@n3, 2=@n6.RGB, 3=@n7, A=@n14) {materialFunction=&Lerp_Multiple_Float3}.Lerp 3 Inputs)
   B=Multiply(A=@n2.A, B=@n1.A)
  )
 ]
}
