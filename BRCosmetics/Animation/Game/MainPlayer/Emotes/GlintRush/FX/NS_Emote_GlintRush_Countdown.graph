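# Niagara countdown effect for the GlintRush emote, showing a flickering, lit "Continue" sprite with digits counting down from 10 to 0, placed and scaled by User parameters.
# NE_Numbers loops eleven times, using particle partition to pick SubUV frames from an array.
# PropSpawn_Glitch opens with a flare and camera-locked glitch meshes tinted by time of day.
# Age-keyed curves intensify and fade the glow near the end, timed to a triangle burst and flashes.
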NiagaraSystem /BRCosmetics/Animation/Game/MainPlayer/Emotes/GlintRush/FX/NS_Emote_GlintRush_Countdown
dependencies {
 usedDataInterfaces [
  &NiagaraDataInterfaceCurve
  &NiagaraDataInterfaceSpriteRendererInfo
  &NiagaraDataInterfaceArrayFloat
  &NiagaraDataInterfaceVector2DCurve
  &NiagaraDataInterfaceCamera
  &NiagaraDataInterfaceColorCurve
 ]
 usedDynamicInputs []
 usedModules [&SystemState]
}
emitters {
 Continue {
  emitterUpdateScript [
   &EmitterState(Life Cycle Mode=::NewEnumerator1, Loop Behavior=::NewEnumerator1, Loop Duration Mode=::NewEnumerator1, UseLoopDelay=false)
   &SpawnBurst_Instantaneous(Use Spawn Probability=false, UseLoopCountLimit=false)
  ]
  particleSpawnScript [
   &PartitionParticles(Number Of Partitions=11)
   &InitializeParticle(
    Position Offset:NiagaraPosition=$User.Continue Offset
    Sprite Size=&Multiply_Vector2D(A=<150 50>, B=&Vector2DFromFloat(Value=$User.ContinueScale))
    Sprite Size Mode=::NewEnumerator4
    UsePositionOffset=true
    Write Material Random=true
   ) {module=InitializeParticle001}
  ]
  particleUpdateScript [
   &ParticleState(Kill Particles When Lifetime Has Elapsed=false, Let Infinitely Lived Particles Die When Emitter Deactivates=true)
   &ScaleSpriteSize(
    Scale Factor=&MakeVector2D(
     X=&FloatFromCurve(
      FloatCurve=DataInterface() {propertyValues {Curve {keys [
          <at 0 0 RCIM_Cubic>
          {arriveTangent=8.9337, interpMode=RCIM_Cubic, leaveTangent=8.9337, time=0.0809, value=1.1}
          {arriveTangent=-0.189, interpMode=RCIM_Cubic, leaveTangent=-0.189, time=0.1119, value=1}
          {arriveTangent=-0.3208, interpMode=RCIM_Cubic, leaveTangent=-0.3208, tangentMode=RCTM_User, time=0.61, value=1}
          <at 0.64 1 RCIM_Cubic>
         ]}}}
      Scale Curve=&Waveform("Waveform [1]"=::NewEnumerator0, "[1] Frequency"=1, "[1] Amplitude Min/Max"=<0.99 1>)
     )
     Y=&FloatFromCurve(
      FloatCurve=DataInterface() {propertyValues {Curve {keys [
          <at 0.017 0 RCIM_Cubic>
          {arriveTangent=11.8228, interpMode=RCIM_Cubic, leaveTangent=11.8228, time=0.0815, value=0.8}
          {arriveTangent=0.3714, interpMode=RCIM_Cubic, leaveTangent=0.3714, time=0.1015, value=1}
          {arriveTangent=0.2037, interpMode=RCIM_Cubic, leaveTangent=0.2037, tangentMode=RCTM_User, time=0.62, value=1}
          <at 0.635 1 RCIM_Cubic>
         ]}}}
      Scale Curve=&Waveform("Waveform [1]"=::NewEnumerator0, "[1] Frequency"=1, "[1] Amplitude Min/Max"=<0.99 1>)
     )
    )
    Scale Sprite Size Mode=::NewEnumerator2
   )
   &SolveForcesAndVelocity(Clamp Velocity=false, Limit Acceleration=false)
   &ScaleColor(
    Scale Alpha=&Waveform(
     "Waveform [1]"=::NewEnumerator0
     "[1] Frequency"=&FloatFromCurve(FloatCurve=DataInterface() {propertyValues {Curve {keys [<at 0.7 1>,<at 9 1.5>]}}}, CurveIndex=$Particles.Age)
     "[1] Amplitude Min/Max"=&MakeVector2D(
      X=&FloatFromCurve(FloatCurve=DataInterface() {propertyValues {Curve {keys [<at 0.7 0.8>,<at 9 0.4>]}}}, CurveIndex=$Particles.Age)
      Y=1.1
     )
    )
    Scale RGB=&VectorFromFloat(Value=&FloatFromCurve(FloatCurve=DataInterface() {propertyValues {Curve {keys [<at 8.5439 1>,<at 8.8039 1.5>]}}}, CurveIndex=$Particles.Age))
    ScaleA=true
    ScaleRGB=true
   )
   &DynamicMaterialParameters(
    Index 0 Param 1=10
    Index 0 Param 2=10
    Index 0 Param 3=&FloatFromCurve(
     FloatCurve=DataInterface() {propertyValues {Curve {keys [
         {arriveTangent=-0.1244, interpMode=RCIM_Cubic, leaveTangent=-0.1244, tangentMode=RCTM_User, time=-0.03, value=0.2}
         {arriveTangent=0.0066, interpMode=RCIM_Cubic, leaveTangent=0.0066, tangentMode=RCTM_User, time=0.6421, value=-0.4}
         <at 8.64 -0.4>
         <at 9 0.2>
        ]}}}
     CurveIndex=$Particles.Age
    )
    Index 0 Param 4=0.01
    Index 1 Param 1=1
    Index 1 Param 2=4.7007
    Index 1 Param 3=1
    Index 1 Param 4=1
    Module.Index0Float0Write=true
    Module.Index0Float1Write=true
    Module.Index0Float2Write=true
    Module.Index0Float3Write=true
    Module.Index1Float0Write=true
    Module.Index1Float1Write=true
    Module.Index1Float2Write=true
    Module.Index1Float3Write=true
    Write Parameter Index 1=::NewEnumerator1
   )
   /SetVariables(
    Particles.GlowColor=&MultiplyLinearColor(
     Linear Color=<rgba 1 0.94 0.3596 0.07>
     Scale Factor=&MakeVector4(
      X=1
      Y=1
      Z=1
      W=&Multiply_Float(
       A=&Waveform(
        "Waveform [1]"=::NewEnumerator0
        "[1] Frequency"=&FloatFromCurve(FloatCurve=DataInterface() {propertyValues {Curve {keys [<at 0.7 1>,<at 9 1.5>]}}}, CurveIndex=$Particles.Age)
        "[1] Amplitude Min/Max"=&MakeVector2D(
         X=&FloatFromCurve(FloatCurve=DataInterface() {propertyValues {Curve {keys [<at 0.7 0.8>,<at 9 0.4>]}}}, CurveIndex=$Particles.Age)
         Y=1.1
        )
       )
       B=&FloatFromCurve(FloatCurve=DataInterface() {propertyValues {Curve {keys [<at 0.1797 0>,<at 0.5 0.9841>,<at 8.64 1>,<at 8.9 0>]}}}, CurveIndex=$Particles.Age)
      )
     )
    )
    Particles.GlowPosition:NiagaraPosition=&AddVectorToPosition(Position=$Particles.Initial.Position, Vector=<4 0 2>)
    Particles.GlowSize=&Multiply_Vector2D(A=<190 40>, B=&Vector2DFromFloat(Value=$User.ContinueScale))
    Particles.LightColor=&MultiplyLinearColor(
     Linear Color=<rgba 1 0.6415 0.1401 1>
     Scale Factor=&MakeVector4(
      X=1
      Y=1
      Z=1
      W=&FloatFromCurve(FloatCurve=DataInterface() {propertyValues {Curve {keys [<at 0 0>,<at 0.049 1>,<at 0.8927 1>,<at 1 0>]}}})
     )
    )
   ) {isSetParameters=true}
   /SetVariables(Particles.Vector 4:Vector4f=<1 1 1 1>) {isSetParameters=true}
  ]
  rendererClasses [&NiagaraSpriteRendererProperties, &NiagaraLightRendererProperties]
  renderers [
   {rendererClass=&NiagaraSpriteRendererProperties}
   {rendererClass=&NiagaraSpriteRendererProperties, rendererIndex=1}
   {rendererClass=&NiagaraLightRendererProperties, rendererIndex=2}
  ]
 }
 Flash001 {
  emitterUpdateScript [
   &EmitterState(Life Cycle Mode=::NewEnumerator1, Loop Behavior=::NewEnumerator1, Loop Delay=0.15, Loop Duration=10, UseLoopDelay=false)
   &SpawnBurst_Instantaneous(Spawn Count=4, Spawn Time=8.75, Use Spawn Probability=false)
  ]
  particleSpawnScript [
   &PartitionParticles(Number Of Partitions=4)
   &InitializeParticle(
    Color=<rgba 1 0.7702 0.4271 1>
    Lifetime=0.17
    Position Offset:NiagaraPosition=$User.Continue Offset
    Sprite Size Mode=::NewEnumerator3
    Uniform Sprite Size=&Multiply_Float(A=30, B=1)
    UsePositionOffset=true
    Write Material Random=true
   )
   &ShapeLocation(MaskSpawnGroup=false, Override Seed=false, Sphere Radius=0, UseOwnerScale=true)
   &CameraOffset(Camera Offset Amount=-10)
   /SetVariables(Particles.SizeWideFlare:Vector2f=&Multiply_Vector2D_ByFloat(Float=1.5, Vector2D=<225 40>)) {isSetParameters=true}
   &InitialMeshOrientation(Mesh Orientation Mode=::NewEnumerator2, Orientation Vector=<0 1 0>, Rotation=<0.25 0 0>, Use Rotation Vector=true)
  ]
  particleUpdateScript [
   &ParticleState()
   &SolveForcesAndVelocity(Clamp Velocity=false, Limit Acceleration=false)
   &ScaleColor(
    Scale Alpha=&FloatFromCurve(FloatCurve=DataInterface() {propertyValues {Curve {keys [
         <at 0.1 0 RCIM_Cubic>
         {interpMode=RCIM_Cubic, leaveTangent=-2.0997, tangentMode=RCTM_Break, time=0.2521, value=1}
         {arriveTangentWeight=0.7369, interpMode=RCIM_Cubic, tangentMode=RCTM_Break, time=1, value=1}
        ]}}})
    Scale RGB=&VectorFromFloat(Value=&FloatFromCurve(
      FloatCurve=DataInterface() {propertyValues {Curve {keys [
          <at 0.1 0 RCIM_Cubic>
          {interpMode=RCIM_Cubic, leaveTangent=-2.0997, tangentMode=RCTM_Break, time=0.2521, value=1}
          {arriveTangentWeight=0.7369, interpMode=RCIM_Cubic, tangentMode=RCTM_Break, time=0.6, value=1}
         ]}}}
      Scale Curve=5
     ))
    ScaleA=true
    ScaleRGB=true
   )
   &ScaleSpriteSize(Uniform Curve Sprite Scale=DataInterface() {propertyValues {Curve {keys [
        <at 0.3 0 RCIM_Cubic>
        {interpMode=RCIM_Cubic, tangentMode=RCTM_User, time=0.4, value=1}
        {interpMode=RCIM_Cubic, tangentMode=RCTM_User, time=1, value=0}
       ]}}})
   &FN_Sprite_Attributes(
    Color=&MakeLinearColorFromVectorAndFloat(
     "Vector (RGB)"=&MakeVectorFromLinearColorRGB(Color=<rgba 3 2.5757 0.4844 1>)
     "Float (Alpha)"=&Multiply_Float(
      A=1
      B=&FloatFromCurve(FloatCurve=DataInterface() {propertyValues {Curve {keys [{interpMode=RCIM_Cubic, tangentMode=RCTM_User, time=0, value=1}, {interpMode=RCIM_Cubic, tangentMode=RCTM_User, time=1, value=0.5}]}}})
     )
    )
    Create Color Attribute=true
    Write to Module Namespace=true
   )
   /SetVariables(Particles.SizeWideFlare:Vector2f=&Multiply_Vector2D(
     A=$Particles.Initial.SizeWideFlare
     B=&MakeVector2D(
      X=&FloatFromCurve(FloatCurve=DataInterface() {propertyValues {Curve {keys [<at 0.1 0>,<at 0.2 0.4>,<at 0.3 1>,<at 1 0>]}}})
      Y=&FloatFromCurve(FloatCurve=DataInterface() {propertyValues {Curve {keys [<at 0.2 1>,<at 0.5 0.1>]}}})
     )
    )) {isSetParameters=true}
  ]
  rendererClasses [&NiagaraSpriteRendererProperties, &NiagaraMeshRendererProperties]
  renderers [
   {rendererClass=&NiagaraSpriteRendererProperties}
   {rendererClass=&NiagaraSpriteRendererProperties, rendererIndex=1}
   {rendererClass=&NiagaraMeshRendererProperties, rendererIndex=2}
  ]
 }
 NE_Numbers {
  emitterUpdateScript [
   &EmitterState(Life Cycle Mode=::NewEnumerator1, Loop Behavior=::NewEnumerator2, Loop Count=11, Loop Duration=0.83, UseLoopDelay=false)
   &SpawnBurst_Instantaneous(Use Spawn Probability=false, UseLoopCountLimit=false)
  ]
  particleSpawnScript [
   &PartitionParticles(Number Of Partitions=11)
   &InitializeParticle(
    Position Offset=&Add_Vector(A=$User.Number Offset, B=Linked() {linkedVariable {name=User.Continue Offset, type=NiagaraPosition}})
    Sprite Size Mode=::NewEnumerator3
    Uniform Sprite Size=&Multiply_Float(A=30, B=$User.Number Scale)
    UsePositionOffset=true
    Write Material Random=true
   ) {module=InitializeParticle001}
   &CameraOffset(Camera Offset Amount=1)
  ]
  particleUpdateScript [
   &ParticleState()
   &ScaleSpriteSize(
    Uniform Curve Scale=&Waveform("Waveform [1]"=::NewEnumerator0, "[1] Frequency"=1, "[1] Amplitude Min/Max"=<0.99 1.04>)
    Uniform Curve Sprite Scale=DataInterface() {propertyValues {Curve {keys [{tangentMode=RCTM_User, time=0, value=1.3}, <at 0.0379 1>, {tangentMode=RCTM_User, time=1, value=1}]}}}
   )
   &SubUVAnimation(
    Direct Frame Index:NiagaraFloat=&SelectFloatFromArray(
     Direct Array Index=$Particles.PartitionParticles.Partition
     Float Selection Array=DataInterface() {propertyValues {FloatData [10,9,8,7,6,5,4,3,2,1,0]}}
    )
    "Sprite Renderer "=DataInterface() {propertyValues {SpriteRenderer=&self:NE_Numbers_0.NiagaraSpriteRendererProperties_0}}
    SubUV Animation Mode:ENiagara_SubUVLookupModeV2=::NewEnumerator5
   )
   &SolveForcesAndVelocity(Clamp Velocity=false, Limit Acceleration=false)
   &ScaleColor(ScaleA=true, ScaleRGB=true)
   /SetVariables(
    Particles.GlowColor=&MultiplyLinearColor(
     Linear Color=<rgba 1 0.94 0.3596 1>
     Scale Factor=&MakeVector4(
      X=1
      Y=1
      Z=1
      W=&FloatFromCurve(
       FloatCurve=DataInterface() {propertyValues {Curve {
          defaultValue=1
          keys [
           {interpMode=RCIM_Cubic, tangentMode=RCTM_User, time=0, value=0}
           {interpMode=RCIM_Cubic, tangentMode=RCTM_User, time=0.3109, value=1}
           {arriveTangent=-1.4513, interpMode=RCIM_Cubic, leaveTangent=-1.4513, time=0.7823, value=1}
           {tangentMode=RCTM_User, time=1, value=0}
          ]
         }}}
       Scale Curve=&FloatFromCurve(FloatCurve=DataInterface() {propertyValues {Curve {keys [<at 8.64 1>,<at 8.9 0>]}}}, CurveIndex=$Emitter.Age, Scale Curve=0.03)
      )
     )
    )
    Particles.GlowSize=&Multiply_Vector2D_ByFloat(Float=$User.Number Scale, Vector2D=<30 40>)
   ) {isSetParameters=true}
   &DynamicMaterialParameters(
    Index 0 Param 1=15
    Index 0 Param 2=15
    Index 0 Param 3=&Lerp_Float(
     A=&FloatFromCurve(FloatCurve=DataInterface() {propertyValues {Curve {keys [<at 0 -0.5>,<at 0.7 1>,<at 0.8291 1>,<at 0.9972 -0.5>]}}})
     B=-0.528
     Alpha=&FloatFromCurve(FloatCurve=DataInterface() {propertyValues {Curve {keys [<at 8.64 0>,<at 8.9 1>]}}}, CurveIndex=$Emitter.Age)
    )
    Index 0 Param 4=0.136
    Index 1 Param 1=1
    Index 1 Param 2=5.213
    Index 1 Param 3=3
    Index 1 Param 4=4
    Module.Index0Float0Write=true
    Module.Index0Float1Write=true
    Module.Index0Float2Write=true
    Module.Index0Float3Write=true
    Module.Index1Float0Write=true
    Module.Index1Float1Write=true
    Module.Index1Float2Write=true
    Module.Index1Float3Write=true
    Write Parameter Index 1=::NewEnumerator1
   )
   /SetVariables(Particles.Vector 4:Vector4f=<1 1 1 1>) {isSetParameters=true}
  ]
  rendererClasses [&NiagaraSpriteRendererProperties]
  renderers [
   {rendererClass=&NiagaraSpriteRendererProperties}
   {rendererClass=&NiagaraSpriteRendererProperties, rendererIndex=1}
  ]
 }
 PropSpawn_Glitch {
  emitterSpawnScript [/SetVariables(
    Emitter.ColorA:LinearColor=<rgba 1 0.9471 0.651 1>
    Emitter.ColorB:LinearColor=<rgba 1 0.8956 0.3119 1>
    Emitter.MasterSizeMultiplier=0.6
    Emitter.ShapeLocationScale:Vector3f=&VectorFromFloat(Value=$User.ContinueScale)
   ) {isSetParameters=true}]
  emitterUpdateScript [
   &EmitterState(Life Cycle Mode=::NewEnumerator1, Loop Behavior=::NewEnumerator1, Loop Delay=0.1, Loop Duration=1, UseLoopDelay=true)
   &SpawnBurst_Instantaneous(Use Spawn Probability=false, UseLoopCountLimit=false)
   &SpawnBurst_Instantaneous(Spawn Count=15, Spawn Group=2, Use Spawn Probability=false, UseLoopCountLimit=false) {module=SpawnBurst_Instantaneous008}
   &SpawnBurst_Instantaneous(Spawn Count=15, Spawn Group=2, Spawn Time=0.15, Use Spawn Probability=false, UseLoopCountLimit=false) {module=SpawnBurst_Instantaneous012}
   &SpawnBurst_Instantaneous(Spawn Count=6, Spawn Group=3, Use Spawn Probability=false, UseLoopCountLimit=false) {module=SpawnBurst_Instantaneous011}
   &SpawnBurst_Instantaneous(Spawn Count=6, Spawn Group=3, Spawn Time=0.05, Use Spawn Probability=false, UseLoopCountLimit=false) {module=SpawnBurst_Instantaneous010}
   &SpawnBurst_Instantaneous(Spawn Count=20, Spawn Group=3, Spawn Time=0.1, Use Spawn Probability=false, UseLoopCountLimit=false) {module=SpawnBurst_Instantaneous009}
  ]
  particleSpawnScript [
   /SetVariables(Particles.VisibilityTag=&SpawnGroup()) {isSetParameters=true}
   /SetVariables(Particles.Random Range:NiagaraFloat=&RandomRangeFloat(Minimum=0.2, Maximum=0.8)) {isSetParameters=true}
   &InitializeParticle(
    Color Mode=::NewEnumerator0
    Lifetime=&MergeHelper_FloatSwitch(
     0=0.27
     1=0.3
     2=&RandomRangeFloat(Minimum=0.25, Maximum=0.3)
     3=&RandomRangeFloat(Minimum=0.25, Maximum=0.3)
     4=0.2
     6=&RandomRangeFloat(Minimum=0.55, Maximum=0.55)
     7=&RandomRangeFloat(Minimum=0.5, Maximum=0.62)
     8=&RandomRangeFloat(Minimum=0.25, Maximum=0.5)
     9=&RandomRangeFloat(Minimum=0.25, Maximum=0.5)
     Index=&SpawnGroup()
     Num=::NewEnumerator3
    )
    Position Offset:NiagaraPosition=$User.Continue Offset
    UsePositionOffset=true
    Write Material Random=true
   ) {module=InitializeParticle001}
   &CameraOffset(Camera Offset Amount=&MergeHelper_FloatSwitch(0=2, 1=2, 2=10, 3=10, 4=2, Index=0, Num=::NewEnumerator3))
   /SetVariables(Particles.Bright Flare Size:Vector2f=<300 100>, Particles.Bright Flare color:LinearColor=$Emitter.ColorB) {isSetParameters=true}
   /SetVariables(
    Particles.Long Glitch Dynam 1:Vector4f=&MakeVector4(X=0, Y=0, Z=0, W=&RandomRangeFloat(Minimum=-1, Maximum=0.1))
    Particles.Long Glitch Dynam 2:Vector4f=&MakeVector4(X=&RandomRangeFloat(Minimum=-100, Maximum=100), Y=0, Z=0, W=0)
    Particles.Long Glitch Size:Vector3f=&RandomRangeVector(Minimum=<0.5 3 0.2>, Maximum=<0.1 8 0.2>)
   ) {isSetParameters=true}
   /SetVariables(
    Particles.Square Glitch Dynam 1:Vector4f=&MakeVector4(X=0, Y=0, Z=0, W=&RandomRangeFloat(Minimum=-1, Maximum=0.1))
    Particles.Square Glitch Dynam 2:Vector4f=&MakeVector4(X=&RandomRangeFloat(Minimum=-100, Maximum=100), Y=0, Z=0, W=0)
    Particles.Square Glitch Size:Vector3f=&RandomRangeVector(Minimum=<0.4 0.4 0.4>, Maximum=<1 1 1>)
   ) {isSetParameters=true}
   &ShapeLocation(
    Box Size=<30 140 30>
    MaskSpawnGroup=false
    Non Uniform Scale=$Emitter.ShapeLocationScale
    Override Seed=false
    Shape Origin=&Subtract_Vector(A=Linked() {linkedVariable {name=User.Continue Offset, type=NiagaraPosition}}, B=<0 0 -1>)
    Shape Primitive=::NewEnumerator2
    Surface Only=false
    UseOwnerScale=true
   )
   &AddVelocity(
    Velocity=&MaskVectorBySpawnGroup(Masked Vector=$Particles.ShapeLocation.ShapeNormal, Passthrough Vector=<0 0 0>, Spawn Group Mask=4)
    Velocity Speed Scale=&Multiply_Float(A=&RandomRangeFloat(Minimum=50, Maximum=80), B=$Emitter.MasterSizeMultiplier)
   )
  ]
  particleUpdateScript [
   &ParticleState()
   /SetVariables(Particles.TODColor:LinearColor=&ColorFromCurve(
     ColorCurve=DataInterface() {propertyValues {
       AlphaCurve {keys [<at 0 1 RCIM_Cubic>]}
       BlueCurve {keys [
         <at 0.0043 0.7>
         {arriveTangent=-0.1359, interpMode=RCIM_Cubic, leaveTangent=-0.1359, time=0.1789, value=0.7}
         <at 0.2812 0.6624>
         {arriveTangent=1.2198, interpMode=RCIM_Cubic, leaveTangent=1.2198, time=0.3527, value=1}
         {arriveTangent=-1.0217, interpMode=RCIM_Cubic, leaveTangent=-1.0217, time=0.558, value=1}
         {arriveTangent=-0.9848, interpMode=RCIM_Cubic, leaveTangent=-0.9848, time=0.8029, value=0.54}
         <at 0.8615 0.7011>
         <at 1 0.7>
        ]}
       GreenCurve {keys [
         <at 0.0043 0.4211>
         {arriveTangent=0.9841, interpMode=RCIM_Cubic, leaveTangent=0.9841, time=0.1789, value=0.5785}
         <at 0.2812 0.6936>
         {arriveTangent=1.1069, interpMode=RCIM_Cubic, leaveTangent=1.1069, time=0.3527, value=1}
         {arriveTangent=-0.8536, interpMode=RCIM_Cubic, leaveTangent=-0.8536, time=0.558, value=1}
         {arriveTangent=-1.3917, interpMode=RCIM_Cubic, leaveTangent=-1.3917, time=0.8029, value=0.6157}
         <at 0.8615 0.5776>
         <at 1 0.4138>
        ]}
       RedCurve {keys [
         <at 0.0043 0.3653>
         {arriveTangent=1.9379, interpMode=RCIM_Cubic, leaveTangent=1.9379, time=0.1789, value=0.5542}
         <at 0.2812 0.902>
         {arriveTangent=0.3542, interpMode=RCIM_Cubic, leaveTangent=0.3542, time=0.3527, value=1}
         {arriveTangent=-0.2221, interpMode=RCIM_Cubic, leaveTangent=-0.2221, time=0.558, value=1}
         {arriveTangent=-1.476, interpMode=RCIM_Cubic, leaveTangent=-1.476, time=0.8029, value=0.9}
         <at 0.8615 0.552>
         <at 1 0.3566>
        ]}
      }}
     CurveIndex=Linked() {linkedVariable {name=NPC.FortniteNPC.RemappedTimeOfDay_0-1, type=NiagaraFloat}}
    )) {isSetParameters=true}
   /SetVariables(
    Particles.BrightFlarePosition:NiagaraPosition=$User.Continue Offset
    Particles.Scale Bright Flare Size:Vector2f=&Multiply_Vector2D_ByFloat(
     Float=$Emitter.MasterSizeMultiplier
     Vector2D=&Multiply_Vector2D(
      A=$Particles.Initial.Bright Flare Size
      B=&Vector2DFromFloat(Value=&FloatFromCurve(FloatCurve=DataInterface() {propertyValues {Curve {keys [
            <at 0 -0.0447>
            <at 0.0141 1.0065>
            <at 0.0287 -0.0508>
            {arriveTangent=-0.3125, interpMode=RCIM_Cubic, leaveTangent=-0.3125, tangentMode=RCTM_User, time=0.0532, value=1}
            {arriveTangent=-3.7966, interpMode=RCIM_Cubic, leaveTangent=-3.7966, tangentMode=RCTM_User, time=1, value=0}
           ]}}}))
     )
    )
    Particles.Update Bright Flare Color:LinearColor=&MakeLinearColorFromVectorAndFloat(
     "Vector (RGB)"=&MakeVectorFromLinearColorRGB(Color=&MultiplyLinearColor(
       Linear Color=$Particles.Initial.Bright Flare color
       Scale Factor=&MakeVector4_FromLinearColor(LinearColor=$Particles.TODColor)
      ))
     "Float (Alpha)"=&FloatFromCurve(FloatCurve=DataInterface() {propertyValues {Curve {keys [
          {arriveTangent=-2.225, interpMode=RCIM_Cubic, leaveTangent=-2.225, tangentMode=RCTM_User, time=0.1, value=1}
          {interpMode=RCIM_Cubic, tangentMode=RCTM_User, time=1, value=0}
         ]}}})
    )
   ) {isSetParameters=true}
   /SetVariables(
    Particles.Long Glitch Dynam 1:Vector4f=&MakeVector4(
     X=0
     Y=&RandomRangeFloat(Minimum=-1, Maximum=1)
     Z=&FloatFromCurve(FloatCurve=DataInterface() {propertyValues {Curve {keys [
          <at 0 2 RCIM_Cubic>
          {arriveTangent=-6.2285, interpMode=RCIM_Cubic, leaveTangent=-6.2285, tangentMode=RCTM_User, time=0.5612, value=1.0841}
          {arriveTangent=-0.8021, interpMode=RCIM_Cubic, leaveTangent=-0.8021, tangentMode=RCTM_User, time=1, value=0}
         ]}}})
     W=0
    )
    Particles.Long Glitch Dynam 2:Vector4f=&MakeVector4(
     X=0
     Y=0
     Z=0
     W=&FloatFromCurve(FloatCurve=DataInterface() {propertyValues {Curve {keys [
          {arriveTangent=4.1785, interpMode=RCIM_Cubic, leaveTangent=4.1785, tangentMode=RCTM_User, time=0, value=0}
          {arriveTangent=0.9995, interpMode=RCIM_Cubic, leaveTangent=0.9995, tangentMode=RCTM_User, time=0.2223, value=0.6825}
          {arriveTangent=-0.0405, interpMode=RCIM_Cubic, leaveTangent=-0.0405, tangentMode=RCTM_User, time=1, value=1}
         ]}}})
    )
    Particles.Scale Long Glitch Size:Vector3f=&Multiply_VectorByFloat(
     Vector=&Multiply_Vector(
      A=$Particles.Initial.Long Glitch Size
      B=&Multiply_VectorByFloat(
       Vector=<1 1 1>
       Float=&Multiply_Float(
        A=0.08
        B=&FloatFromCurve(
         FloatCurve=DataInterface() {propertyValues {Curve {keys [
             {arriveTangent=5.4245, interpMode=RCIM_Cubic, leaveTangent=5.4245, tangentMode=RCTM_User, time=0, value=5}
             {arriveTangent=-3.9614, interpMode=RCIM_Cubic, leaveTangent=-3.9614, time=0.0963, value=8.5355}
             {arriveTangent=-22.2662, interpMode=RCIM_Cubic, leaveTangent=-22.2662, time=0.2702, value=3.9295}
             <at 0.3 4>
             <at 0.4 11.5>
             {arriveTangent=0.1532, interpMode=RCIM_Cubic, leaveTangent=0.1532, tangentMode=RCTM_User, time=0.5, value=2.875}
             <at 0.6 3>
             <at 0.7 11.5>
             <at 0.8 2.5>
             <at 0.9 2>
             {arriveTangent=-0.0405, interpMode=RCIM_Cubic, leaveTangent=-0.0405, tangentMode=RCTM_User, time=1, value=0.5}
            ]}}}
         Scale Curve=&Multiply_Float(
          A=0.6
          B=&ClampFloat(
           Float=&FN_LOD_Distance_Remap(Start Distance=200, End Distance=400, Minimum Output=0.5, Maximum Output=1)
           Min=0.5
          )
         )
        )
       )
      )
     )
     Float=$Emitter.MasterSizeMultiplier
    )
    Particles.Update Long Glitch Color:LinearColor=&HueShiftLinearColor(
     Hue=&RandomRangeFloat(Minimum=0.85, Maximum=1)
     Saturation=&RandomRangeFloat(Minimum=3, Maximum=1)
     Value=&RandomRangeFloat(Minimum=0.7, Maximum=1)
     Linear Color=&MultiplyLinearColor(Linear Color=$Emitter.ColorA, Scale Factor=&MakeVector4_FromLinearColor(LinearColor=$Particles.TODColor))
    )
   ) {isSetParameters=true}
   /SetVariables(
    Particles.Scale Square Glitch Size:Vector3f=&Multiply_VectorByFloat(
     Vector=&Multiply_Vector(
      A=$Particles.Initial.Square Glitch Size
      B=&Multiply_VectorByFloat(
       Vector=<1 1 1>
       Float=&Multiply_Float(
        A=0.08
        B=&FloatFromCurve(
         FloatCurve=DataInterface() {propertyValues {Curve {keys [
             {arriveTangent=5.4245, interpMode=RCIM_Cubic, leaveTangent=5.4245, tangentMode=RCTM_User, time=0, value=5}
             {arriveTangent=-3.9614, interpMode=RCIM_Cubic, leaveTangent=-3.9614, time=0.0963, value=8.5355}
             {arriveTangent=-22.2662, interpMode=RCIM_Cubic, leaveTangent=-22.2662, time=0.2702, value=3.9295}
             <at 0.3 4>
             <at 0.4 11.5>
             {arriveTangent=0.1532, interpMode=RCIM_Cubic, leaveTangent=0.1532, tangentMode=RCTM_User, time=0.5, value=2.875}
             <at 0.6 3>
             <at 0.7 11.5>
             <at 0.8 2.5>
             <at 0.9 2>
             {arriveTangent=-0.0405, interpMode=RCIM_Cubic, leaveTangent=-0.0405, tangentMode=RCTM_User, time=1, value=0.5}
            ]}}}
         Scale Curve=&Multiply_Float(
          A=0.6
          B=&ClampFloat(
           Float=&FN_LOD_Distance_Remap(Start Distance=200, End Distance=400, Minimum Output=0.5, Maximum Output=1)
           Min=0.5
          )
         )
        )
       )
      )
     )
     Float=$Emitter.MasterSizeMultiplier
    )
    Particles.Square Glitch Dynam 1:Vector4f=&MakeVector4(
     X=0
     Y=&RandomRangeFloat(Minimum=-1, Maximum=1)
     Z=&FloatFromCurve(FloatCurve=DataInterface() {propertyValues {Curve {keys [
          <at 0 2 RCIM_Cubic>
          {arriveTangent=-6.2285, interpMode=RCIM_Cubic, leaveTangent=-6.2285, tangentMode=RCTM_User, time=0.5612, value=1.0841}
          {arriveTangent=-0.8021, interpMode=RCIM_Cubic, leaveTangent=-0.8021, tangentMode=RCTM_User, time=1, value=0}
         ]}}})
     W=0
    )
    Particles.Square Glitch Dynam 2:Vector4f=&MakeVector4(
     X=&RandomRangeFloat(Minimum=-1, Maximum=1)
     Y=0
     Z=0
     W=&FloatFromCurve(FloatCurve=DataInterface() {propertyValues {Curve {keys [
          {arriveTangent=4.1785, interpMode=RCIM_Cubic, leaveTangent=4.1785, tangentMode=RCTM_User, time=0, value=0}
          {arriveTangent=0.9995, interpMode=RCIM_Cubic, leaveTangent=0.9995, tangentMode=RCTM_User, time=0.2223, value=0.6825}
          {arriveTangent=-0.0405, interpMode=RCIM_Cubic, leaveTangent=-0.0405, tangentMode=RCTM_User, time=1, value=1}
         ]}}})
    )
    Particles.Update Square Glitch Color:LinearColor=&HueShiftLinearColor(
     Hue=&RandomRangeFloat(Minimum=0.85, Maximum=1)
     Saturation=1
     Value=1
     Linear Color=&MultiplyLinearColor(Linear Color=$Emitter.ColorB, Scale Factor=&MakeVector4_FromLinearColor(LinearColor=$Particles.TODColor))
    )
   ) {isSetParameters=true}
   &Drag(
    Drag=&MaskFloatBySpawnGroup(
     Masked Float=&FloatFromCurve(
      FloatCurve=DataInterface() {propertyValues {Curve {keys [
          {arriveTangent=2.0771, interpMode=RCIM_Cubic, leaveTangent=2.0771, tangentMode=RCTM_User, time=0, value=0}
          {interpMode=RCIM_Cubic, tangentMode=RCTM_User, time=1, value=1}
         ]}}}
      Scale Curve=10
     )
     Passthrough Float=0
     Spawn Group Mask=4
    )
    Rotational Drag:NiagaraFloat=&MaskFloatBySpawnGroup(
     Masked Float=&FloatFromCurve(
      FloatCurve=DataInterface() {propertyValues {Curve {keys [
          {arriveTangent=2.0771, interpMode=RCIM_Cubic, leaveTangent=2.0771, tangentMode=RCTM_User, time=0, value=0}
          {interpMode=RCIM_Cubic, tangentMode=RCTM_User, time=1, value=1}
         ]}}}
      Scale Curve=3
     )
     Passthrough Float=0
     Spawn Group Mask=4
    )
    Use Linear Drag=true
    Use Rotational Drag=true
   )
   &SolveForcesAndVelocity(Clamp Velocity=false, Limit Acceleration=false)
   &self:GimbalLock_01(Gimbal Lock Direction:NiagaraPosition=&ReturnCameraVector(Camera Vectors=::NewEnumerator4, Coordinate Space=::Local, Origin Position=$Engine.Emitter.SimulationPosition)) {module=GimbalLock_01}
  ]
  rendererClasses [&NiagaraSpriteRendererProperties, &NiagaraMeshRendererProperties]
  renderers [
   {rendererClass=&NiagaraSpriteRendererProperties}
   {rendererClass=&NiagaraMeshRendererProperties, rendererIndex=1}
   {rendererClass=&NiagaraMeshRendererProperties, rendererIndex=2}
  ]
 }
 Tris_Continue {
  emitterUpdateScript [
   &EmitterState(Life Cycle Mode=::NewEnumerator1, Loop Duration=15, UseLoopDelay=false)
   &SpawnBurst_Instantaneous(Spawn Count=100, Spawn Time=8.6, Use Spawn Probability=false, UseLoopCountLimit=false)
  ]
  particleSpawnScript [
   &InitializeParticle(
    Color=<rgba 1 0.6696 0.0957 1>
    Lifetime Max=0.7
    Lifetime Min=0.3
    Lifetime Mode=::NewEnumerator1
    Mass Max=1.5
    Mass Mode=::NewEnumerator3
    Mesh Scale Mode=::NewEnumerator1
    Mesh Uniform Scale Max:NiagaraFloat=&FloatFromCurve(
     FloatCurve=DataInterface() {propertyValues {Curve {keys [<at 0 0 RCIM_Constant>,<at 0.5 0.4 RCIM_Constant>,<at 0.8685 0.8 RCIM_Constant>]}}}
     CurveIndex=&RandomRangeFloat(Minimum=0, Maximum=1)
     Scale Curve=0.4
    )
    Mesh Uniform Scale Min=0.1
    Sprite Rotation Mode=::NewEnumerator2
    Sprite Size Mode=::NewEnumerator3
    UsePositionOffset=true
    Write Material Random=true
   )
   &InitialMeshOrientation(Use Rotation Vector=false)
   &MeshRotationForce(Rotation=&RandomRangeVector(Minimum=<1 1 1>, Maximum=<3 3 3>))
   &ApplyInitialForces()
   &ShapeLocation(
    Box Size=&Multiply_VectorByFloat(Vector=<25 130 5>, Float=$User.ContinueScale)
    MaskSpawnGroup=false
    Override Seed=false
    Rotation Mode=::NewEnumerator1
    Shape Origin=$User.Continue Offset
    Shape Primitive=::NewEnumerator2
    Surface Only=false
    UseOwnerScale=true
   )
  ]
  particleUpdateScript [
   &ParticleState()
   &ScaleMeshSize(Scale Factor=&VectorFromFloat(Value=&FloatFromCurve(FloatCurve=DataInterface() {propertyValues {Curve {keys [
          {interpMode=RCIM_Cubic, tangentMode=RCTM_User, time=0, value=0}
          {interpMode=RCIM_Cubic, tangentMode=RCTM_User, time=0.5, value=1}
          {tangentMode=RCTM_User, time=1, value=1}
         ]}}})))
   &SolveForcesAndVelocity(Clamp Velocity=false, Limit Acceleration=false)
   &DynamicMaterialParameters(
    Index 0 Param 1=20
    Index 0 Param 2=5
    Index 0 Param 3=&RandomRangeFloat(Minimum=0, Maximum=0.2)
    Index 0 Param 4=&RandomRangeFloat(Minimum=0, Maximum=0.2)
    Module.Index0Param0Write=true
    Module.Index0Param1Write=true
    Module.Index0Param2Write=true
    Module.Index0Param3Write=true
   ) {module=DynamicMaterialParameters001}
   &ScaleColor(Scale Alpha=&RandomRangeFloat(Minimum=0.85, Maximum=1), ScaleA=true, ScaleRGB=true)
   /SetVariables(
    Particles.Linear Color=/Niagara/DynamicInputs/Add/Add_LinearColor(A:LinearColor=$Particles.Color, B:LinearColor=<rgba 0 0 0 -0.8>)
    Particles.SpriteSize=&Multiply_Vector2D_ByFloat(
     Float=20
     Vector2D=&MakeVector2D(
      X=&MakeFloatFromVector(Vector=$Particles.Scale, Channel=::NewEnumerator0)
      Y=&MakeFloatFromVector(Vector=$Particles.Scale, Channel=::NewEnumerator1)
     )
    )
   ) {isSetParameters=true}
   &ScaleColor(
    Scale Alpha=&FloatFromCurve(FloatCurve=DataInterface() {propertyValues {Curve {keys [
         {interpMode=RCIM_Cubic, tangentMode=RCTM_User, time=0, value=0}
         {interpMode=RCIM_Cubic, tangentMode=RCTM_User, time=0.5, value=1}
         {tangentMode=RCTM_User, time=1, value=0}
        ]}}})
    ScaleA=true
    ScaleRGB=true
   ) {module=ScaleColor001}
   /SetVariables(Particles.Vector 4:Vector4f=&MakeVector4(X=0, Y=0, Z=0, W=0)) {isSetParameters=true}
   /SetVariables(Particles.BoxSection:NiagaraInt32=&MakeCustomIntFromBool(
     Bool=&SetBoolByFloatComparison(
      A=&MakeFloatFromVector(Vector=Linked() {linkedVariable {name=Particles.Initial.Position, type=NiagaraPosition}}, Channel=::NewEnumerator2)
      B=&MakeFloatFromVector(
       Vector=&Add_Vector(
        A=Linked() {linkedVariable {name=Engine.Emitter.SimulationPosition, type=NiagaraPosition}}
        B=Linked() {linkedVariable {name=User.Continue Offset, type=NiagaraPosition}}
       )
       Channel=::NewEnumerator2
      )
     )
     True INT=1
     False INT=0
    )) {isSetParameters=true}
   /SetVariables(Particles.Position=&AddVectorToPosition(
     Position=$Particles.Initial.Position
     Vector=&Multiply_Vector(
      A=&VectorFromFloat(Value=&FloatFromCurve(
        FloatCurve=DataInterface() {propertyValues {Curve {
           defaultValue=0
           keys [
            {arriveTangent=4.1785, interpMode=RCIM_Cubic, leaveTangent=4.1785, tangentMode=RCTM_User, time=0, value=0}
            {arriveTangent=1.1179, interpMode=RCIM_Cubic, leaveTangent=1.1179, tangentMode=RCTM_User, time=0.2042, value=0.6228}
            {arriveTangent=0.0224, interpMode=RCIM_Cubic, leaveTangent=0.0224, tangentMode=RCTM_User, time=1, value=1}
           ]
          }}}
        Scale Curve=&RandomRangeFloat(Minimum=0.9, Maximum=1)
       ))
      B=&Multiply_Vector(
       A=&MaskVector(Mask Source=1, Mask Parameter=$Particles.BoxSection, Mask=0, Masked Vector=<0 0 20>, Passthrough Vector=<1 1 -20>)
       B=&VectorFromFloat(Value=$User.ContinueScale)
      )
     )
    )) {isSetParameters=true}
   &self:GimbalLock_01(Gimbal Lock Direction:NiagaraPosition=&ReturnCameraVector(Camera Vectors=::NewEnumerator4, Coordinate Space=::Local, Origin Position=$Engine.Emitter.SimulationPosition)) {module=GimbalLock_01}
  ]
  rendererClasses [&NiagaraMeshRendererProperties]
  renderers [{rendererClass=&NiagaraMeshRendererProperties, rendererIndex=0}]
 }
}
summary {
 systemName=NS_Emote_GlintRush_Countdown
 emitters [
  {emitterName=NE_Numbers, rendererClasses [&NiagaraSpriteRendererProperties]}
  {emitterName=Continue, rendererClasses [&NiagaraSpriteRendererProperties, &NiagaraLightRendererProperties]}
  {emitterName=Tris_Continue, rendererClasses [&NiagaraMeshRendererProperties]}
  {
   emitterName=PropSpawn_Glitch
   rendererClasses [&NiagaraSpriteRendererProperties, &NiagaraMeshRendererProperties]
  }
  {emitterName=Flash001, rendererClasses [&NiagaraSpriteRendererProperties, &NiagaraMeshRendererProperties]}
 ]
 userVariables [
  User.bIsFrontEnd {type=NiagaraBool}
  User.bIsFrontEndPreview {type=NiagaraBool}
  User.Continue Offset {type=NiagaraPosition, description="", defaultValue {}}
  User.ContinueScale {type=NiagaraFloat, defaultValue=0.75, description=""}
  User.Number Offset {type=Vector3f, description="", defaultValue {struct=Vector3f, value=<0 0 -25>}}
  User.Number Scale {type=NiagaraFloat, defaultValue=0.75, description=""}
 ]
}
